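# Niagara idle effect for the left one-handed Plum pickaxe: a mesh sweep and a sprite flare.
# FN_Sweep bursts one short-lived mesh particle, ramps a dynamic material parameter, brightens sharply near its end and generates a death event.
# FN_Flare is a spinning, camera-offset sprite, raised above the system origin, that pops and fades.
# Its own burst is disabled and no event handler appears here.
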NiagaraSystem /BRCosmetics/Weapons/FORT_Melee/Pickaxe_Plum_1h/FX/NS_Pickaxe_Plum_1H_L_Idle
dependencies {usedDataInterfaces [&NiagaraDataInterfaceCurve], usedDynamicInputs [], usedModules [&SystemState]}
emitters {
 FN_Flare {
  emitterUpdateScript [
   &EmitterState(Life Cycle Mode=::NewEnumerator1, Loop Duration=1, UseLoopDelay=false)
   &SpawnBurst_Instantaneous(Spawn Count=1, Use Spawn Probability=false) {enabled=false}
  ]
  particleSpawnScript [
   /Niagara/Modules/Spawn/Initialization/InitializeParticle(
    Color=<rgba 0.4306 0.6707 0.7188 2>
    Lifetime=&RandomRangeFloat(Minimum=0.5, Maximum=0.45)
    Module.Write Color=true
    Module.Write Lifetime=true
    Module.Write Mass=true
    Module.Write Material Random=true
    Module.Write Position=false
    Module.Write Scale=false
    Module.Write SpriteRotation=true
    Module.Write SpriteSize=true
    Position=&SimulationPosition()
    Sprite Rotation=&RandomRangeFloat(Minimum=-1, Maximum=1)
    Sprite Size=<26 22>
   )
   &SystemLocation(MaskSpawnGroup=false, Offset=<-0.1 0 60>)
   &CameraOffset(Camera Offset Amount=15)
  ]
  particleUpdateScript [
   &ParticleState()
   &SolveForcesAndVelocity(Clamp Velocity=false, Limit Acceleration=false)
   &SpriteRotationRate(Rotation Rate=&Multiply_Float(A=&RandomRangeFloat(Minimum=0.8, Maximum=1.2), B=360))
   &ScaleSpriteSize(Scale Factor=&Vector2DFromFloat(Value=&FloatFromCurve(FloatCurve=DataInterface() {propertyValues {Curve {keys [
          {interpMode=RCIM_Cubic, tangentMode=RCTM_User, time=0, value=0}
          {arriveTangent=1.7059, interpMode=RCIM_Cubic, leaveTangent=1.7059, tangentMode=RCTM_User, time=0.05, value=0.94}
          {arriveTangent=-1.3315, interpMode=RCIM_Cubic, leaveTangent=-1.3315, tangentMode=RCTM_User, time=0.34, value=0.58}
          <at 1 0>
         ]}}})))
   &ScaleColor(
    Scale RGB=&VectorFromFloat(Value=&FloatFromCurve(FloatCurve=DataInterface() {propertyValues {Curve {keys [<at 0 5>,<at 0.3 1>,<at 1 0>]}}}))
    ScaleA=true
    ScaleRGB=true
   )
  ]
  rendererClasses [&NiagaraSpriteRendererProperties]
  renderers [{rendererClass=&NiagaraSpriteRendererProperties}]
 }
 FN_Sweep {
  emitterUpdateScript [
   &EmitterState()
   &SpawnRate(SpawnRate=0, Use Spawn Probability=false) {enabled=false}
   &SpawnBurst_Instantaneous(Spawn Count=1, Spawn Time=0.5, Use Spawn Probability=false)
  ]
  particleSpawnScript [
   &InitializeParticle(
    Color=<rgba 0.33 0.4425 0.6 0.35>
    Lifetime=0.3
    Module.Write Color=true
    Module.Write Lifetime=true
    Module.Write Mass=true
    Module.Write Material Random=true
    Module.Write Position=false
    Module.Write Scale=false
    Module.Write SpriteRotation=false
    Module.Write SpriteSize=true
    Position=&SimulationPosition()
    Sprite Size=<5 5>
   )
   &CameraOffset(Camera Offset Amount=1)
  ]
  particleUpdateScript [
   &ParticleState()
   &ScaleColor(
    Scale RGB=&VectorFromFloat(Value=&FloatFromCurve(FloatCurve=DataInterface() {propertyValues {Curve {keys [
          {interpMode=RCIM_Cubic, leaveTangent=0.3383, tangentMode=RCTM_Break, time=0, value=0.1}
          {arriveTangent=3.0213, interpMode=RCIM_Cubic, leaveTangent=5.4486, tangentMode=RCTM_Break, time=0.85, value=1.5}
          {arriveTangent=0.0059, interpMode=RCIM_Cubic, leaveTangent=0.0059, tangentMode=RCTM_User, time=0.9, value=10}
         ]}}}))
    ScaleA=true
    ScaleRGB=true
   )
   &DynamicMaterialParameters(
    Index 0 Param 1=&FloatFromCurve(FloatCurve=DataInterface() {propertyValues {Curve {keys [
         {arriveTangent=1.2112, interpMode=RCIM_Cubic, leaveTangent=1.2112, tangentMode=RCTM_User, time=0, value=-0.1}
         {arriveTangent=3.7988, interpMode=RCIM_Cubic, leaveTangent=3.7988, tangentMode=RCTM_User, time=1, value=2}
        ]}}})
    Module.Param0WriteEnabled=true
    Module.Param1WriteEnabled=true
    Module.Param2WriteEnabled=true
    Module.Param3WriteEnabled=true
   )
   &SolveForcesAndVelocity(Clamp Velocity=false, Limit Acceleration=false)
   &GenerateDeathEvent()
  ]
  rendererClasses [&NiagaraMeshRendererProperties]
  renderers [{rendererClass=&NiagaraMeshRendererProperties, rendererIndex=0}]
 }
}
summary {
 systemName=NS_Pickaxe_Plum_1H_L_Idle
 emitters [
  {emitterName=FN_Sweep, rendererClasses [&NiagaraMeshRendererProperties]}
  {emitterName=FN_Flare, rendererClasses [&NiagaraSpriteRendererProperties]}
 ]
 userVariables []
}
